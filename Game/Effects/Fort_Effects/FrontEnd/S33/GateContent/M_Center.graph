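# Additive, unlit Niagara sprite material for the center of the Season 33 front-end gate effect, drawing a distorted, rotating, circular tech-tile pattern.
# Panning radial noise, modulated by temporal dither, distorts bump-offset UVs, with a second sphere-mask-driven parallax before sampling the tile texture.
# Dynamic parameters drive rotation, depth offset and noise-based alpha erosion.
# Particle color tints the result, while sphere masks clip the alpha and darken the color toward the center.
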Material /Game/Effects/Fort_Effects/FrontEnd/S33/GateContent/M_Center
graph {
 properties {bUsedWithNiagaraSprites=true, blendMode=BLEND_Additive, pixelDepthOffsetMode=PDOM_Legacy, shadingModel=MSM_Unlit, usedShadingModels=Unlit}
 nodes [
  n1 = ParticleColor()
  n2 = DynamicParameter() {defaultValue=<rgba 1 0 1 0.1>, paramNames [Time,AlphaOffset,Param3,DepthOffset]}
  n3 = BumpOffset(
   Coordinate=TextureCoordinate()
   Height=Constant()
   HeightRatioInput=DynamicParameter() {defaultValue=<rgba 1 0 1 0.1>, paramNames [Time,AlphaOffset,Param3,DepthOffset]}.DepthOffset
  )
  n4 = MaterialFunctionCall(Vector or UVs=Add(A=@n3) {constB=-0.5}) {materialFunction=&VectorToRadialValue}
  n5 = VectorParameter() {defaultValue=<rgba 1 2 0.1 0.1>}
  n6 = AppendVector(A=@n5.R, B=@n5.R)
  n7 = TextureSample!(UVs=Panner(Coordinate=Multiply(A=@n4.Radial Coordinates, B=@n6), Speed=AppendVector(A=@n5.B, B=@n5.B)), Tex, Level, Apply View MipBias) {
   mipValueMode=TMVM_MipLevel
   samplerType=SAMPLERTYPE_LinearColor
   texture=&/T_FX_Tile_0138_Moon
  }
  n8 = MaterialFunctionCall() {materialFunction=&DitherTemporalAA}
  n9 = Multiply(A=@n7.RGB, B=Multiply(A=ScalarParameter() {defaultValue=0.3287, parameterName=Distort}, B=Multiply(A=@n8.Result, B=@n7.B) {constB=0.2}))
  n10 = Add(A=@n3, B=ComponentMask(None=@n9) {g=true, r=true})
  n11 = Constant() {r=0.5}
  n12 = SphereMask(A=@n3, B=@n11, Radius=@n11) {hardnessPercent=0}
  n13 = TextureSampleParameter2D!(
   UVs=BumpOffset(Coordinate=Rotator(Coordinate=@n10, Time=@n2.Time), Height=ComponentMask(None=@n9) {b=true}, HeightRatioInput=@n12) {referencePlane=0}
   Bias=ScalarParameter() {defaultValue=4, parameterName=Param_6}
   Apply View MipBias
  ) {
   mipValueMode=TMVM_MipBias
   parameterName=Param
   samplerType=SAMPLERTYPE_LinearColor
   texture=&/Tile_Tech_02
  }
  n14 = Multiply(A=@n1.A, B=Power(Base=@n13.A, Exp=ScalarParameter() {defaultValue=3.8265, parameterName=Param_2}))
  n15 = Multiply(
   A=DynamicParameter() {defaultValue=<rgba 1 0 1 0.1>, paramNames [Time,AlphaOffset,Param3,DepthOffset]}.AlphaOffset
   B=TextureSample!(UVs, Tex, Apply View MipBias) {samplerType=SAMPLERTYPE_LinearColor, texture=&T_FX_Tile_0138_Moon}.G
  )
  Multiply(
   A=SphereMask(
    A=TextureCoordinate()
    B=Constant() {r=0.5}
    Radius=ScalarParameter() {defaultValue=0.4, parameterName=Spheremask}
    Hardness=ScalarParameter() {defaultValue=100, parameterName=SphereMaskHardness}
   )
   B=Saturate(None=Subtract(A=@n14, B=@n15))
  )
  n16 = Multiply(A=@n13.G, B=ScalarParameter() {defaultValue=2.1127, parameterName=Param_1})
  n17 = OneMinus(None=@n12)
  n18 = ScalarParameter() {defaultValue=1, parameterName=Param_5}
  Power(Base=Multiply(A=Multiply(A=@n1.RGB, B=@n16), B=Power(Base=@n17, Exp=@n18)), Exp=ScalarParameter() {defaultValue=2.4898})
  ScalarParameter() {defaultValue=1, parameterName=Param_3}
  ScalarParameter() {defaultValue=2, parameterName=Param_4}
 ]
}
