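# Unlit, translucent, two-sided ribbon material for Niagara mesh particles in the Season 33 front-end gate effects.
# Panned main and detail textures blend into a mask, alongside two-color emissive tinting, particle color and alpha, and a view-angle Fresnel.
# World position offset uses SplineThicken for width and gradient vector noise for displacement, with a ramp data texture available as a mask.
# Dynamic parameters drive main animation, detail panning, per-mesh noise and per-mesh width.
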Material /Game/Effects/Fort_Effects/FrontEnd/S33/GateContent/M_Ribbon
graph {
 properties {
  bUsedWithNiagaraMeshParticles=true
  blendMode=BLEND_Translucent
  pixelDepthOffsetMode=PDOM_Legacy
  shadingModel=MSM_Unlit
  twoSided=true
  usedShadingModels=Unlit
 }
 nodes [
  n1 = DynamicParameter() {defaultValue=<rgba 1 1 50 1>, paramNames [Animation, DetialTexturePanner, PerMesh_Noise, PerMesh_Width]}
  n2 = ComponentMask(None=NamedRerouteUsage()) {r=true}
  n3 = MaterialFunctionCall() {materialFunction=&ObjectScale}
  n4 = Multiply(
   A=MaterialFunctionCall!(RangeMin=Constant(), RangeMax=Constant() {r=1}, Value=NamedRerouteUsage()) {materialFunction=/Engine/Functions/Engine_MaterialFunctions03/Math/WithinRangeFloat2}.AllComponentsWithinRange
   B=Multiply(A=@n1.PerMesh_Width, B=Multiply(A=@n2, B=@n3.Scale XYZ))
  ) {constA=1}
  n5 = WorldPosition()
  n6 = NamedRerouteUsage()
  Add(
   A=MaterialFunctionCall(WidthBase=@n4, WidthTip=@n4, WorldPosition=Add(A=@n5.XYZ, B=@n6)) {materialFunction=&SplineThicken}.WorldPositionOffset
   B=NamedRerouteUsage()
  )
  n7 = NamedRerouteUsage()
  n8 = VertexColor()
  Multiply(A=Multiply(A=Multiply(A=@n7, B=@n8.A), B=ParticleColor().A), B=NamedRerouteUsage())
  n9 = ScalarParameter() {defaultValue=100, parameterName=NoiseScale}
  Multiply(A=@n9, B=MaterialFunctionCall() {materialFunction=&ObjectScale}.Scale XYZ)
  n10 = TextureSampleParameter2D(UVs=NamedRerouteUsage()) {
   group=Main
   parameterName=MainTexture
   samplerSource=SSM_Clamp_WorldGroupSettings
   texture=&/Trail_RibbonTrail
  }
  n11 = ChannelMaskParameter(None=@n10.RGBA) {group=Emissive, parameterName=MainTexChannel}
  n12 = TextureSampleParameter2D(UVs=NamedRerouteUsage()) {
   group=Detail
   parameterName=DetailTexture
   samplerType=SAMPLERTYPE_LinearColor
   texture=&/Tile_Directional_T_LightArc
  }
  n13 = Multiply(A=@n11.R, B=ChannelMaskParameter(None=@n12.RGBA) {group=Emissive, parameterName=DetailTexChannel}.R)
  n14 = ScalarParameter() {defaultValue=1, group=Main, parameterName="DetialTexBlend(Multiply)"}
  NamedRerouteDeclaration(None=Saturate(None=LinearInterpolate(A=@n11.R, B=@n13, Alpha=@n14))) {name=blendedTexture, nodeColor=<rgba 0.0339 1 0 1>}
  n15 = TextureCoordinate()
  n16 = MaterialFunctionCall(XY=@n15) {materialFunction=&Swizzle}
  n17 = ScalarParameter() {defaultValue=1, group=Main, parameterName=Main_UV_Swizzle}
  n18 = VectorParameter() {group=Main, parameterName=Main_UV, defaultValue=<rgba -1 1 1 0>}
  n19 = DynamicParameter() {defaultValue=<rgba 1 1 50 1>, paramNames [Animation, DetialTexturePanner, PerMesh_Noise, PerMesh_Width]}
  NamedRerouteDeclaration(None=MaterialFunctionCall!(ClampTexture, UV=LinearInterpolate(A=@n15, B=@n16.YX, Alpha=@n17), Tex, InRGB=@n18.RGB, InA=@n18.RGB, Time=@n19.Animation, TimeMult) {materialFunction=&/MF_MOON_TexPanner_V2}.PannerUV) {name=MainUV, nodeColor=<rgba 0 0.4049 1 1>}
  n20 = VertexTangentWS()
  n21 = VertexNormalWS()
  n22 = Subtract(A=CameraPositionWS(), B=WorldPosition().XYZ)
  n23 = Abs(None=DotProduct(A=CrossProduct(A=@n20, B=@n21), B=Normalize(VectorInput=@n22)))
  n24 = OneMinus(None=@n23)
  n25 = ScalarParameter() {parameterName="Fresnel_Invert (0or1)"}
  n26 = ScalarParameter() {defaultValue=0.6, parameterName=Fresnel_Offset}
  n27 = Subtract(A=Power(Base=LinearInterpolate(A=@n24, B=@n23, Alpha=@n25), Exp=ScalarParameter() {defaultValue=0.2, parameterName=Fresnel_Power}), B=@n26)
  n28 = Add(A=@n26)
  NamedRerouteDeclaration(None=VertexInterpolator(VS=Multiply(A=@n27, B=@n28)).PS) {name=Fresnel, nodeColor=<rgba 1 0 0.9613 1>}
  NamedRerouteDeclaration(None=@n11.R) {name=MainTexture, nodeColor=<rgba 0.2187 1 0 1>}
  n29 = LinearInterpolate(
   A=VectorParameter() {group=Emissive, parameterName=EmissiveColorTintA, defaultValue=<rgba 0.0461 0.0707 0.15 1>}.RGB
   B=VectorParameter() {group=Emissive, parameterName=EmissiveColorTintB, defaultValue=<rgba 0 1 0 1>}.RGB
   Alpha=NamedRerouteUsage()
  )
  NamedRerouteDeclaration(None=Multiply(A=ParticleColor().RGB, B=Saturate(None=@n29))) {name=Emissive, nodeColor=<rgba 0 0 0 1>}
  n30 = Divide(
   A=TransformPosition(None=WorldPosition() {worldPositionShaderOffset=WPT_ExcludeAllShaderOffsets}.XYZ) {transformSourceType=TRANSFORMPOSSOURCE_World}
   B=@n9
  )
  n31 = Multiply(
   A=ComponentMask(None=VectorNoise(World Position=@n30) {noiseFunction=VNF_GradientALU}) {b=true, g=true, r=true}
   B=ComponentMask(None=NamedRerouteUsage()) {g=true}
  )
  n32 = DynamicParameter() {defaultValue=<rgba 1 1 50 1>, paramNames [Animation, DetialTexturePanner, PerMesh_Noise, PerMesh_Width]}
  NamedRerouteDeclaration(None=Multiply(
    A=Multiply(A=@n31, B=@n32.PerMesh_Noise)
    B=MaterialFunctionCall() {materialFunction=&ObjectScale}.Scale XYZ
   )) {name=NoiseValue, nodeColor=<rgba 0.0502 1 0 1>}
  n33 = TextureSampleParameter2D(UVs=NamedRerouteUsage()) {
   parameterName=DataTexture
   samplerSource=SSM_Clamp_WorldGroupSettings
   texture=/Game/Global/Textures/Default/Masks_and_Gradients/T_Soft_Ramp_Vert_03
  }
  NamedRerouteDeclaration(None=@n33.RGB) {name=RampData, nodeColor=<rgba 0 0 1 1>}
  n34 = TextureCoordinate()
  n35 = MaterialFunctionCall(XY=@n34) {materialFunction=&Swizzle}
  n36 = ScalarParameter() {defaultValue=1, group=Detail, parameterName=Detail_UV_Swizzle}
  n37 = VectorParameter() {group=Detail, parameterName=Detail_UV, defaultValue=<rgba 2 0.5 0.5 0>}
  n38 = DynamicParameter() {defaultValue=<rgba 1 1 50 1>, paramNames [Animation, DetialTexturePanner, PerMesh_Noise, PerMesh_Width]}
  NamedRerouteDeclaration(None=MaterialFunctionCall(UV=LinearInterpolate(A=@n34, B=@n35.YX, Alpha=@n36), InRGB=@n37.RGB, InA=@n37.RGB, Time=@n38.DetialTexturePanner) {materialFunction=&MF_MOON_TexPanner_V2}.PannerUV) {name=DetailUV, nodeColor=<rgba 0 0.4049 1 1>}
  NamedRerouteUsage()
  NamedRerouteUsage()
 ]
}
